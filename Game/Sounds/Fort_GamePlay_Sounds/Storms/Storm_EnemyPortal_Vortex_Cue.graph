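# Looping sound cue for the storm enemy portal vortex, in the weather sound class with vortex-specific attenuation.
# A quality level root steps down variety: a random pick of four loops, then of two, then one fixed loop shared by the last two levels.
# Each branch passes through a modulator that randomizes pitch upward to as much as double.
# A local concurrency override caps simultaneous instances, superseding the assigned concurrency set.
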SoundCue /Game/Sounds/Fort_GamePlay_Sounds/Storms/Storm_EnemyPortal_Vortex_Cue
root=QualityLevel_0
nodes [
 {id=Modulator_0, children [Random_0], properties {pitchMax=2, pitchMin=1}}
 {id=Modulator_1, children [Random_1], properties {pitchMax=2, pitchMin=1}}
 {id=Modulator_2, children [WavePlayer_22], properties {pitchMax=2, pitchMin=1}}
 {id=QualityLevel_0, children [Modulator_0,Modulator_1,Modulator_2,Modulator_2]}
 {id=Random_0, children [WavePlayer_16, WavePlayer_17, WavePlayer_18, WavePlayer_19], properties {weights [1,1,1,1]}}
 {id=Random_1, children [WavePlayer_20, WavePlayer_21], properties {weights [1,1]}}
 {id=WavePlayer_16, properties {bLooping=true, soundWaveAssetPtr=&/Storm_Spawn_EnemyVortex_01_Loop}}
 {id=WavePlayer_17, properties {bLooping=true, soundWaveAssetPtr=&/Storm_Spawn_EnemyVortex_02_Loop}}
 {id=WavePlayer_18, properties {bLooping=true, soundWaveAssetPtr=&/Storm_Spawn_EnemyVortex_03_Loop}}
 {id=WavePlayer_19, properties {bLooping=true, soundWaveAssetPtr=&/Storm_Spawn_EnemyVortex_04_Loop}}
 {id=WavePlayer_20, properties {bLooping=true, soundWaveAssetPtr=&Storm_Spawn_EnemyVortex_01_Loop}}
 {id=WavePlayer_21, properties {bLooping=true, soundWaveAssetPtr=&Storm_Spawn_EnemyVortex_04_Loop}}
 {id=WavePlayer_22, properties {bLooping=true, soundWaveAssetPtr=&Storm_Spawn_EnemyVortex_04_Loop}}
]
properties {
 attenuationSettings=/Game/Sounds/Attenuation/Fort_Storm_Vortex_Att
 bOverrideConcurrency=true
 maxDistance=1100
 soundClassObject=/Game/Sounds/FortSoundClassesAndModes/Fort_10_Weather_Related
 volumeMultiplier=0.7
 concurrencyOverrides {maxCount=3}
 concurrencySet [/Game/Sounds/Concurrency_Rules/FiveMaximum_StopFurthest]
}
