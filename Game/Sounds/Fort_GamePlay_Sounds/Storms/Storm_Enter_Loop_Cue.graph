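# Sound cue for entering the storm, layering an interior enter sound with an interior storm loop on the storm sound class.
# A quality-level node picks among mixer branches, with the two lowest tiers sharing one built from "opt" waves.
# Each mixer pairs the enter wave with a loop wave whose envelope fades in, then eases down.
# The "opt" loop player lacks the looping flag, and the concurrency override prevents new overlapping instances.
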SoundCue /Game/Sounds/Fort_GamePlay_Sounds/Storms/Storm_Enter_Loop_Cue
root=QualityLevel_0
nodes [
 {
  id=Enveloper_0
  children [WavePlayer_0]
  properties {
   pitchCurve {editorCurveData {keys [{time=0}]}}
   volumeCurve {editorCurveData {keys [<at 0.0302 -0.0141>,<at 1.4584 1.3881>,<at 6.0054 1.0728>,<at 7.5295 0.5268>]}}
  }
 }
 {
  id=Enveloper_1
  children [WavePlayer_4]
  properties {
   pitchCurve {editorCurveData {keys [{time=0}]}}
   volumeCurve {editorCurveData {keys [<at 0.0302 -0.0141>,<at 1.4584 1.3881>,<at 6.0054 1.0728>,<at 7.5295 0.5268>]}}
  }
 }
 {
  id=Enveloper_2
  children [WavePlayer_3]
  properties {
   pitchCurve {editorCurveData {keys [{time=0}]}}
   volumeCurve {editorCurveData {keys [<at 0.0302 -0.0141>,<at 1.4584 1.3881>,<at 6.0054 1.0728>,<at 7.5295 0.5268>]}}
  }
 }
 {id=Mixer_0, children [WavePlayer_1,Enveloper_0], properties {inputVolume [1.2,1.5]}}
 {id=Mixer_1, children [WavePlayer_5,Enveloper_1], properties {inputVolume [1.2,1.5]}}
 {id=Mixer_2, children [WavePlayer_2,Enveloper_2], properties {inputVolume [1.2,1.5]}}
 {id=QualityLevel_0, children [Mixer_0,Mixer_1,Mixer_2,Mixer_2]}
 {id=WavePlayer_0, properties {bLooping=true, soundWaveAssetPtr=&/storm_interior_loop}}
 {id=WavePlayer_1, properties {soundWaveAssetPtr=&/storm_interior_enter}}
 {id=WavePlayer_2, properties {soundWaveAssetPtr=&/storm_interior_enter_opt}}
 {id=WavePlayer_3, properties {soundWaveAssetPtr=&/storm_interior_loop_opt}}
 {id=WavePlayer_4, properties {bLooping=true, soundWaveAssetPtr=&storm_interior_loop}}
 {id=WavePlayer_5, properties {soundWaveAssetPtr=&storm_interior_enter}}
]
properties {
 bOverrideConcurrency=true
 maxDistance=2097152
 soundClassObject=/Game/Sounds/FortSoundClassesAndModes/Fort_Storms_PSM
 volumeMultiplier=1
 concurrencyOverrides {resolutionRule=PreventNew}
}
