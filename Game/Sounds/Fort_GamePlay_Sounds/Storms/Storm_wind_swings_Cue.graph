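# Storm ambience sound cue that plays looping wind-and-chain sounds, named for swings blowing in a storm.
# A Random node picks one of two equally weighted looping waves, and a Modulator above it varies pitch, with a raised upper limit.
# It plays at reduced volume through the storm trees attenuation and the ambient tile mix sound class.
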SoundCue /Game/Sounds/Fort_GamePlay_Sounds/Storms/Storm_wind_swings_Cue
root=Modulator_0
nodes [
 {id=Modulator_0, children [Random_0], properties {pitchMax=1.5}}
 {id=Random_0, children [WavePlayer_1,WavePlayer_2], properties {weights [1,1]}}
 {id=WavePlayer_1, properties {bLooping=true, soundWaveAssetPtr=&/storm_wind_chains_01}}
 {id=WavePlayer_2, properties {bLooping=true, soundWaveAssetPtr=&/storm_wind_chains_02}}
]
properties {
 attenuationSettings=/Game/Sounds/Attenuation/Fort_Storm_Trees_Att
 maxDistance=10000
 soundClassObject=/Game/Sounds/FortSoundClassesAndModes/Fort_Ambient_TileMix
 volumeMultiplier=0.5
}
