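# Sound cue for swinging chain impacts, stored with the storm gameplay sounds and routed to the impacts sound class with prop-impact attenuation.
# A quality-level root picks among four modulated branches: random picks from six waves, random picks from four, or, for two levels, a single optimized wave.
# The single-wave branches widen pitch modulation to add variation.
# Concurrency caps it at two per owner, stopping the oldest.
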SoundCue /Game/Sounds/Fort_GamePlay_Sounds/Storms/swing_chain_impact_Cue
root=QualityLevel_0
nodes [
 {id=Modulator_0, children [WavePlayer_0], properties {pitchMax=1.12, pitchMin=0.88}}
 {id=Modulator_1, children [Random_1]}
 {id=Modulator_2, children [Random_2]}
 {id=Modulator_3, children [WavePlayer_2], properties {pitchMax=1.12, pitchMin=0.88}}
 {id=QualityLevel_0, children [Modulator_1,Modulator_2,Modulator_0,Modulator_3]}
 {
  id=Random_1
  children [WavePlayer_6, WavePlayer_7, WavePlayer_8, WavePlayer_9, WavePlayer_10, WavePlayer_11]
  properties {weights [1,1,1,1,1,1]}
 }
 {id=Random_2, children [WavePlayer_13, WavePlayer_14, WavePlayer_15, WavePlayer_16], properties {weights [1,1,1,1]}}
 {id=WavePlayer_0, properties {soundWaveAssetPtr=&/swing_chain_impact_opt}}
 {id=WavePlayer_10, properties {soundWaveAssetPtr=&/swing_chain_impact_05}}
 {id=WavePlayer_11, properties {soundWaveAssetPtr=&/swing_chain_impact_06}}
 {id=WavePlayer_13, properties {soundWaveAssetPtr=&/swing_chain_impact_02}}
 {id=WavePlayer_14, properties {soundWaveAssetPtr=&/swing_chain_impact_03}}
 {id=WavePlayer_15, properties {soundWaveAssetPtr=&/swing_chain_impact_04}}
 {id=WavePlayer_16, properties {soundWaveAssetPtr=&swing_chain_impact_05}}
 {id=WavePlayer_2, properties {soundWaveAssetPtr=&swing_chain_impact_opt}}
 {id=WavePlayer_6, properties {soundWaveAssetPtr=&/swing_chain_impact_01}}
 {id=WavePlayer_7, properties {soundWaveAssetPtr=&swing_chain_impact_02}}
 {id=WavePlayer_8, properties {soundWaveAssetPtr=&swing_chain_impact_03}}
 {id=WavePlayer_9, properties {soundWaveAssetPtr=&swing_chain_impact_04}}
]
properties {
 attenuationSettings=/Game/Sounds/Attenuation/Fort_Prop_Impact_Att
 maxDistance=3256
 soundClassObject=&Fort_Impacts
 volumeMultiplier=1
 concurrencySet [/Game/Sounds/Concurrency_Rules/TwoMaximum_Owner_StopOldest]
}
